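# Animation Blueprint for the FierceBraid backpack, animating the pack as its matching pickaxe is equipped.
# A state machine moves from closed pose through opening to idle loop while that pickaxe is equipped.
# Chained additive transforms then adjust the pack bone, weighted by owner spine/thigh twist and movement, parachute and front-end states.
# Weapon, skydiving and customization events toggle a material hide parameter and pack offset when the pickaxe is held outside skydiving.
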AnimBlueprint /BRCosmetics/Accessories/FORT_Backpacks/Backpack_F_MED_FierceBraid/Meshes/F_MED_FierceBraid_Pack_AnimBP
target_skeleton=&/F_MED_FierceBraid_Pack_Skeleton
graphs [
 AnimGraph {nodes [
   Input Pose()
   AnimGraphNode_StateMachine_0 = PACK_Anim() {is=title AnimGraphNode_StateMachine}
   n1 = "Transform (Modify) Bone - Bone: pack"(
    ComponentPose=Local To Component(LocalPose=@AnimGraphNode_StateMachine_0)
    Alpha=Get Rotation_spine5_z_inSky() -> (Rotation_spine5_z_inSky:float)
    Translation="-0.274,-10.023,-1.560"
    Rotation="-58.344,-21.997,25.389"
   ) {
    rotationMode=BMM_Additive
    rotationSpace=BCS_BoneSpace
    translationMode=BMM_Additive
    alphaBoolBlend {blendOption=Linear}
    boneToModify {boneName=pack}
    rotation=<rot -58.344 -21.997 25.389>
    translation=<-0.274 -10.023 -1.56>
   }
   n2 = "Transform (Modify) Bone - Bone: pack"(
    ComponentPose=@n1
    Alpha=Get Rotation_spine1_z_inSky() -> (Rotation_spine1_z_inSky:float)
    Translation="0,0,0"
    Rotation="-54.067,0.006,-0.007"
   ) {
    rotationMode=BMM_Additive
    rotationSpace=BCS_BoneSpace
    translationMode=BMM_Additive
    alphaBoolBlend {blendOption=Linear}
    boneToModify {boneName=pack}
    rotation=<rot -54.067 0.006 -0.007>
   }
   n3 = "Transform (Modify) Bone - Bone: pack"(
    ComponentPose=@n2
    Alpha=Get Rotation_RightThigh_z_inSky() -> (Rotation_RightThigh_z_inSky:float)
    Translation="0,1,0"
    Rotation="0.520,-2.972,18.685"
   ) {
    rotationMode=BMM_Additive
    rotationSpace=BCS_BoneSpace
    translationMode=BMM_Additive
    alphaBoolBlend {blendOption=Linear}
    boneToModify {boneName=pack}
    rotation=<rot 0.52 -2.972 18.685>
    translation {y=1}
   }
   n4 = "Transform (Modify) Bone - Bone: pack"(
    ComponentPose=@n3
    Alpha=Get ExtraRotation_RightThigh_z_onTheGround() -> (ExtraRotation_RightThigh_z_onTheGround:float)
    Translation="0,0,0"
    Rotation="0,0,-11.83"
   ) {
    rotationMode=BMM_Additive
    rotationSpace=BCS_BoneSpace
    translationMode=BMM_Additive
    translationSpace=BCS_BoneSpace
    alphaBoolBlend {blendOption=Linear}
    alphaScaleBiasClamp {bInterpResult=true}
    boneToModify {boneName=pack}
    rotation {roll=-11.83}
   }
   n5 = "Transform (Modify) Bone - Bone: pack"(
    ComponentPose=@n4
    Alpha=Get Rotation_RightThigh_z_onTheGround() -> (Rotation_RightThigh_z_onTheGround:float)
    Translation="0.000,1.452,-0.373"
    Rotation="-0.000,0.000,12.849"
   ) {
    rotationMode=BMM_Additive
    rotationSpace=BCS_BoneSpace
    translationMode=BMM_Additive
    alphaBoolBlend {blendOption=Linear}
    alphaScaleBiasClamp {bInterpResult=true}
    boneToModify {boneName=pack}
    rotation {roll=12.849}
    translation {y=1.452, z=-0.373}
   }
   n6 = "Transform (Modify) Bone - Bone: pack"(
    ComponentPose=@n5
    Alpha=Get Rotation_spine3_z_InJump() -> (Rotation_spine3_z_InJump:float)
    Translation="0,0,0"
    Rotation="-0.000,0.000,20.155"
   ) {
    rotationMode=BMM_Additive
    rotationSpace=BCS_BoneSpace
    translationMode=BMM_Additive
    alphaBoolBlend {blendOption=Linear}
    boneToModify {boneName=pack}
    rotation {roll=20.155}
   }
   n7 = "Transform (Modify) Bone - Bone: pack"(
    ComponentPose=@n6
    Alpha=Get TransformBackpackWithoutSword() -> (TransformBackpackWithoutSword:float)
    Translation="0.000,3.406,-0.206"
    Rotation="0,0,0"
   ) {
    rotationMode=BMM_Additive
    rotationSpace=BCS_BoneSpace
    translationMode=BMM_Additive
    alphaBoolBlend {blendOption=Linear}
    boneToModify {boneName=pack}
    translation {y=3.406, z=-0.206}
   }
   n8 = "Transform (Modify) Bone - Bone: pack"!(ComponentPose=@n7, bAlphaBoolEnabled:bool=Get bIsParachuteOpened(), Translation="-0.000,-4.989,-0.000", Rotation="0.110,-0.193,2.804", Scale) -> (Pose) {
    alphaInputType=Bool
    rotationMode=BMM_Additive
    rotationSpace=BCS_BoneSpace
    translationMode=BMM_Additive
    alphaBoolBlend {blendOption=Linear}
    boneToModify {boneName=pack}
    rotation=<rot 0.11 -0.193 2.804>
    translation {y=-4.989}
   }
   n9 = "Transform (Modify) Bone - Bone: pack"(ComponentPose=@n8, bAlphaBoolEnabled=Get bIsInFrontEnd(), Translation="0,1,0", Rotation="0,0,0") {
    alphaInputType=Bool
    rotationMode=BMM_Additive
    rotationSpace=BCS_BoneSpace
    translationMode=BMM_Additive
    alphaBoolBlend {blendOption=Linear}
    boneToModify {boneName=pack}
    translation {y=1}
   }
   Output Pose(Result=Component To Local(ComponentPose=@n9)) {layerGroup=None}
  ]}
 PACK_Anim {
  entry=Close
  path=AnimGraph.AnimGraphNode_StateMachine_0.PACK_Anim
  ids=id
  nodes [
   AnimStateNode_1 = Equipped() {is=title AnimStateNode}
   AnimStateNode_2 = Idle()
   AnimStateNode_3 = Close() {is=title AnimStateNode, bAlwaysResetOnEntry=true}
   AnimStateTransitionNode_1 = AnimStateTransitionNode() {bAutomaticRuleBasedOnSequencePlayerInState=true, crossfadeDuration=0.3}
   AnimStateTransitionNode_2 = AnimStateTransitionNode() {crossfadeDuration=0.02}
   AnimStateTransitionNode_4 = AnimStateTransitionNode()
  ]
  links [
   {
    dst=Equipped
    rule=AnimGraph.AnimGraphNode_StateMachine_0.PACK_Anim.AnimStateTransitionNode_4.Transition
    src=Close
    via=AnimStateTransitionNode_4
   }
   {
    dst=Idle
    rule=AnimGraph.AnimGraphNode_StateMachine_0.PACK_Anim.AnimStateTransitionNode_1.Transition
    src=Equipped
    via=AnimStateTransitionNode_1
   }
   {
    dst=Close
    rule=AnimGraph.AnimGraphNode_StateMachine_0.PACK_Anim.AnimStateTransitionNode_2.Transition
    src=Idle
    via=AnimStateTransitionNode_2
   }
  ]
 }
 Equipped {
  path=AnimGraph.AnimGraphNode_StateMachine_0.PACK_Anim.AnimStateNode_1.Equipped
  nodes [Output Animation Pose(Result=Sequence Player 'fierceBriad_Open_PACK'() {bLoopAnimation=false, sequence=/BRCosmetics/Accessories/FORT_Backpacks/Backpack_F_MED_FierceBraid/Anim/fierceBriad_Open_PACK})]
 }
 Idle {
  path=AnimGraph.AnimGraphNode_StateMachine_0.PACK_Anim.AnimStateNode_2.Idle
  nodes [Output Animation Pose(Result=Sequence Player 'fierceBriad_Cycle_PACK'() {sequence=/BRCosmetics/Accessories/FORT_Backpacks/Backpack_F_MED_FierceBraid/Anim/fierceBriad_Cycle_PACK})]
 }
 Close {
  path=AnimGraph.AnimGraphNode_StateMachine_0.PACK_Anim.AnimStateNode_3.Close
  nodes [Output Animation Pose(Result=Sequence Player 'fierceBriad_BasePose_PACK'() {bLoopAnimation=false, sequence=/BRCosmetics/Accessories/FORT_Backpacks/Backpack_F_MED_FierceBraid/Anim/fierceBriad_BasePose_PACK})]
 }
 Transition {path=AnimGraph.AnimGraphNode_StateMachine_0.PACK_Anim.AnimStateTransitionNode_1.Transition, nodes [Result()]}
 Transition {
  path=AnimGraph.AnimGraphNode_StateMachine_0.PACK_Anim.AnimStateTransitionNode_2.Transition
  nodes [Result(bCanEnterTransition=NOT Boolean(A=Get IsFierceBraidEquipped() -> (IsFierceBraidEquipped:bool)))]
 }
 Transition {
  path=AnimGraph.AnimGraphNode_StateMachine_0.PACK_Anim.AnimStateTransitionNode_4.Transition
  nodes [Result(bCanEnterTransition=Get IsFierceBraidEquipped())]
 }
 EventGraph {nodes [
   Event BlueprintUpdateAnimation()
   > n1 = Sequence!(execute) -> (then_0, then_1, then_2:Exec)
   n2 = Branch(execute=@n1.then_1, Condition=OR Boolean(A=Get bIsSkyDiving(), B=Get bIsSliding()))
   > Set ExtraRotation_RightThigh_z_onTheGround(ExtraRotation_RightThigh_z_onTheGround:float=0) -> (Output_Get:float)
   > Set Rotation_RightThigh_z_onTheGround(Rotation_RightThigh_z_onTheGround:float=0) -> (Output_Get:float)
   > Set Rotation_spine3_z_InJump(Rotation_spine3_z_InJump:float=0) -> (Output_Get:float)
   n3 = GetOwnerSkeletalMeshComponent()
   Set ExtraRotation_RightThigh_z_onTheGround(
    execute=@n2.else
    ExtraRotation_RightThigh_z_onTheGround=GetAlphaFromReferenceTransformDelta(SkeletalMeshComponent=@n3, BoneName=thigh_r, BaseName=spine_05, InputRange=(X=-20,Y=-45), TwistAxis="1,0,0")
   )
   > Set Rotation_RightThigh_z_onTheGround(Rotation_RightThigh_z_onTheGround=GetAlphaFromReferenceTransformDelta(SkeletalMeshComponent=@n3, BoneName=thigh_r, BaseName=spine_05, InputRange=(X=0,Y=-24), TwistAxis="1,0,0"))
   > n4 = Branch(Condition=OR Boolean(A=Get bIsFalling(), B=Get bIsJumping()))
   > Set Rotation_spine3_z_InJump(Rotation_spine3_z_InJump=GetAlphaFromReferenceTransformDelta(SkeletalMeshComponent=@n3, BoneName=spine_03, BaseName=thigh_r, InputRange=(X=16,Y=-4), TwistAxis="1,0,0"))
   Set Rotation_spine3_z_InJump(execute=@n4.else, Rotation_spine3_z_InJump=0)
   n5 = PickaxeItemDefinitionHasAnyMetaTag(Tags=Make Array("[0]"=(TagName="Cosmetics.MetaTags.FierceBraid")), bRequiresEquipped=true)
   Set IsFierceBraidEquipped(execute=@n1.then_0, IsFierceBraidEquipped:bool=@n5) -> (Output_Get:bool)
   n6 = Branch(execute=@n1.then_2, Condition=Get bIsSkyDiving())
   n7 = GetOwnerSkeletalMeshComponent()
   > Set Rotation_spine5_z_inSky(Rotation_spine5_z_inSky:float=GetAlphaFromReferenceTransformDelta(SkeletalMeshComponent=@n7, BoneName=spine_05, BaseName=pelvis, InputRange=(X=0,Y=-90), TwistAxis="1,0,0")) -> (Output_Get:float)
   > Set Rotation_spine1_z_inSky(Rotation_spine1_z_inSky:float=GetAlphaFromReferenceTransformDelta(SkeletalMeshComponent=@n7, BoneName=spine_01, BaseName=pelvis, InputRange=(X=0,Y=80), TwistAxis="1,0,0")) -> (Output_Get:float)
   > Set Rotation_RightThigh_z_inSky(Rotation_RightThigh_z_inSky:float=GetAlphaFromReferenceTransformDelta(SkeletalMeshComponent=@n7, BoneName=thigh_r, BaseName=spine_05, InputRange=(X=-5,Y=3), TwistAxis="1,0,0")) -> (Output_Get:float)
   > Set ExtraRotation_RightThigh_z_inSky(ExtraRotation_RightThigh_z_inSky:float=GetAlphaFromReferenceTransformDelta(SkeletalMeshComponent=@n7, BoneName=thigh_r, BaseName=spine_05, InputRange=(X=-5,Y=3), TwistAxis="1,0,0")) -> (Output_Get:float)
   Set Rotation_spine5_z_inSky(execute=@n6.else, Rotation_spine5_z_inSky=0)
   > Set Rotation_spine1_z_inSky(Rotation_spine1_z_inSky=0)
   > Set Rotation_RightThigh_z_inSky(Rotation_RightThigh_z_inSky=0)
   Event BlueprintInitializeAnimation()
   > n8 = Cast To FortPlayerPawn(Object=TryGetPawnOwner())
   > Set As Fort Player Pawn(As Fort Player Pawn:Fort Player Pawn Object Reference=@n8) -> (Output_Get:Fort Player Pawn Object Reference)
   n9 = Event OnWeaponChanged()
   n10 = Get Dynamic Material Instance(execute=[@n9.then, @n15.then], SkeletalMeshComponent=GetPartSkeletalMeshComponent())
   > n11 = Branch(Condition=@n5)
   > n12 = Branch(Condition=Get bIsSkyDiving())
   SetScalarParameterValue(execute=[@n13.then, @n14.else, @n12.then, @n11.else], self=@n10, ParameterName=Hide Element 01, Value=0)
   > Set TransformBackpackWithoutSword(TransformBackpackWithoutSword:float=0) -> (Output_Get:float)
   SetScalarParameterValue(execute=[@n14.then, @n12.else], self=@n10, ParameterName=Hide Element 01, Value=1)
   > Set TransformBackpackWithoutSword(TransformBackpackWithoutSword=1)
   n13 = Event OnBeginSkydiving()
   Event OnEndSkydiving()
   > n14 = Branch(Condition=@n5)
   n15 = Event OnCharacterCustomizationLoaded()
   n16 = Get As Fort Player Pawn() -> (As Fort Player Pawn:Fort Player Pawn Object Reference)
   Contains(
    SearchIn=GetDisplayName(Object=GetCurrentWeapon!(self:Fort Pawn Object Reference=@n16) -> (ReturnValue:Fort Weapon Object Reference))
    Substring=B_Athena_Pickaxe_FierceBraid
   )
  ]}
]
variables [
 As Fort Player Pawn {type=FortPlayerPawn}
 ExtraRotation_RightThigh_z_inSky {type=number}
 ExtraRotation_RightThigh_z_onTheGround {type=number}
 IsFierceBraidEquipped {type=boolean}
 Rotation_RightThigh_z_inSky {type=number}
 Rotation_RightThigh_z_onTheGround {type=number}
 Rotation_spine1_z_inSky {type=number}
 Rotation_spine3_z_InJump {type=number}
 Rotation_spine5_z_inSky {type=number}
 TransformBackpackWithoutSword {type=number, default=1}
]
